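# Niagara system for the Ninjastar vehicle boost cosmetic, trailing ribbons, sparks and spinning star meshes from the car's boost sockets.
# NE_MainRibbon001 and NE_ThrusterSparks001 form the thruster plume, while NE_Ninjastar001 throws stars that fall and die on collision, each trailed by NE_MainRibbon002.
# Scratch modules pass star death positions through system arrays, so DeathParticles spawns sprite bursts there.
# VC_PaintedVariantsOverride supplies painted color variants per emitter.
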NiagaraSystem /VehicleCosmetics/Boosts/Boost_Ninjastar/FX/NS_Boost_Ninjastar
dependencies {
 usedDataInterfaces [
  &NiagaraDataInterfaceCurve
  &NiagaraDataInterfaceSpriteRendererInfo
  &NiagaraDataInterfaceArrayFloat
  &NiagaraDataInterfaceArrayFloat3
  &NiagaraDataInterfaceParticleRead
  &NiagaraDataInterfaceMeshRendererInfo
  &NiagaraDataInterfaceSkeletalMesh
  &NiagaraDataInterfaceArrayPosition
  &NiagaraDataInterfaceArrayColor
 ]
 usedDynamicInputs [
  &SelectFloatFromArray
  &Vector2DFromFloat
  &RandomRangeVector2D
  /Niagara/DynamicInputs/Math/Floor_Float
  &Divide_Float
 ]
 usedModules [
  &self:SystemSpawnScript.SystemScriptSource.SystemScriptGraph.NiagaraNodeAssignment_0.SetVariables_#1
  &SystemState
  &self:ReadDeathArrayScratch_01
 ]
}
emitters {
 DeathParticles {
  emitterUpdateScript [
   &EmitterState()
   &VC_FilteredSocketCount()
   &SpawnRate(SpawnRate=2, Use Spawn Probability=true) {enabled=false}
   &self:SpawnDeathArrayScratch_05(ClearDeathPositions=false) {module=SpawnDeathArrayScratch_05}
   &self:SpawnDeathArrayScratch_05(ClearDeathPositions=false, CountPerPosition=14, Spawn Group=1) {module=SpawnDeathArrayScratch_05001}
  ]
  particleSpawnScript [
   /SetVariables(Particles.VisibilityTag=$Emitter.SpawnGroup) {isSetParameters=true}
   &VC_PaintedVariantsOverride(DefaultColor=<rgba 2.6823 2.8957 5 1>, R1 DefaultColor=<rgba 5 2.605 2.605 1>, Use R1 Painted Overrides=true) {enabled=false}
   &VC_PaintedVariantsOverride(DefaultColor=<rgba 5.2517 1.6276 12.5 3>, R1 DefaultColor=<rgba 8.3333 0 25 25>, Use R1 Painted Overrides=true) {module=VC_PaintedVariantsOverride001}
   &InitializeParticle(
    Color Mode=::NewEnumerator0
    Lifetime=&MergeHelper_FloatSwitch(
     0=&RandomRangeFloat(Minimum=0.55, Maximum=0.65)
     1=&RandomRangeFloat(Minimum=0.2, Maximum=0.25)
     2=&RandomRangeFloat(Minimum=0.2, Maximum=0.3)
     3=&RandomRangeFloat(Minimum=0.2, Maximum=0.3)
     4=&RandomRangeFloat(Minimum=1, Maximum=2)
     Index=$Emitter.SpawnGroup
     Num=::NewEnumerator0
    )
    Mesh Scale Mode=::NewEnumerator3
    Mesh Uniform Scale=&MergeHelper_FloatSwitch(0=0, 1=7.5, 2=8, 3=0, Index=$Emitter.SpawnGroup, Num=::NewEnumerator2)
    Position Offset=&TransformPosition(Position=$System.ReadDeathArrayScratch_01.DeathPosition2, Source Space=::World, Destination Space=::Local)
    Position Offset Coordinate Space=::Local
    Sprite Size Max=&MergeHelper_Vector2Switch(0=<150 150>, 1=<5 30>, 2=?, 3=?, 4=?, Index=$Emitter.SpawnGroup, Num=::NewEnumerator0)
    Sprite Size Min=&MergeHelper_Vector2Switch(0=<75 75>, 1=<3 24>, 2=?, 3=?, 4=?, Index=$Emitter.SpawnGroup, Num=::NewEnumerator0)
    Sprite Size Mode=::NewEnumerator2
    UsePositionOffset=true
    Write Material Random=true
   )
   &PartitionParticles(Number Of Partitions=$Emitter.PortCount) {enabled=false}
   &SkeletalMeshLocation(
    Bone / Socket Index=$Particles.PartitionParticles.Partition
    Bone / Socket Sampling Mode=::NewEnumerator5
    MaskSpawnGroup=false
    Mesh Sampling Type=::NewEnumerator6
    Sampled Position Offset=&MergeHelper_VectorSwitch(
     0=Linked() {linkedVariable {name=System.ReadDeathArrayScratch.DeathPosition2, type=NiagaraPosition}}
     1=<-850 0 0>
     2=<0 0 0>
     3=<0 0 0>
     4=<0 0 0>
     Index=$Emitter.SpawnGroup
     Num=::NewEnumerator3
    )
    Skeletal Mesh=$System.SkeletalMesh
   ) {enabled=false}
   &ShapeLocation(MaskSpawnGroup=true, Offset Mode=::NewEnumerator0, Override Seed=false, Spawn Group Mask=4, UseOwnerScale=true) {module=ShapeLocation001}
   &AddVelocity(
    Module.Use Falloff Distance=false
    Origin Offset:NiagaraPosition=$Particles.ShapeLocation001.ShapePosition
    Override Seed=false
    Velocity Mode=::NewEnumerator1
    Velocity Speed=&MergeHelper_FloatSwitch(
     0=0
     1=&RandomRangeFloat(Minimum=350, Maximum=600)
     2=?
     3=?
     Index=$Emitter.SpawnGroup
     Num=::NewEnumerator0
    )
   )
  ]
  particleUpdateScript [
   &ParticleState()
   &ScaleSpriteSize(
    Scale Sprite Size Mode=::NewEnumerator0
    Uniform Scale Factor=&MergeHelper_FloatSwitch(
     0=&FloatFromCurve(FloatCurve=DataInterface() {propertyValues {Curve {keys [<at 0 1 RCIM_Cubic>,<at 1 0 RCIM_Cubic>]}}})
     1=1
     2=?
     3=&FloatFromCurve(FloatCurve=DataInterface() {propertyValues {Curve {keys [<at 0 1 RCIM_Cubic>,<at 1 0 RCIM_Cubic>]}}})
     Index=&SpawnGroup()
     Num=::NewEnumerator0
    )
   )
   &ScaleColor(
    Scale Alpha=&FloatFromCurve(FloatCurve=DataInterface() {propertyValues {Curve {
        defaultValue=1
        keys [
         {arriveTangent=4.9594, interpMode=RCIM_Cubic, leaveTangent=4.9594, tangentMode=RCTM_User, time=0, value=0}
         {interpMode=RCIM_Cubic, tangentMode=RCTM_User, time=0.5, value=1}
         {tangentMode=RCTM_User, time=1, value=0}
        ]
       }}})
    Scale RGB=&VectorFromFloat(Value=&MergeHelper_FloatSwitch(0=10, 1=150, 2=?, 3=?, Index=&SpawnGroup(), Num=::NewEnumerator0))
    ScaleA=true
    ScaleRGB=true
   )
   &SubUVAnimation(
    Evaluation Type=::NewEnumerator0
    Output Alternate Renderer Binding=true
    "Sprite Renderer "=DataInterface() {propertyValues {SpriteRenderer=&self:DeathParticles_0.NiagaraSpriteRendererProperties_1}}
    SubUV Animation Mode:ENiagara_SubUVLookupModeV2=::NewEnumerator2
    UseEndFrame=false
    UseStartFrame=false
   ) {module=SubUVAnimation002}
   &SubUVAnimation(
    Output Alternate Renderer Binding=true
    "Sprite Renderer "=DataInterface() {propertyValues {SpriteRenderer=&self:NE_VC_Boost_Sprite_Parent.NiagaraSpriteRendererProperties_2}}
    UseEndFrame=false
    UseStartFrame=false
   )
   &GravityForce(Gravity=&MergeHelper_VectorSwitch(
     0=<0 0 0>
     1=&RandomRangeVector(Minimum=<0 0 -600>, Maximum=<0 0 -300>)
     2=?
     3=?
     Index=&SpawnGroup()
     Num=::NewEnumerator0
    ))
   &SolveForcesAndVelocity(Clamp Velocity=false, Limit Acceleration=false)
   /SetVariables(Particles.Vector:NiagaraPosition=$System.ReadDeathArrayScratch_01.DeathPosition2) {isSetParameters=true}
  ]
  rendererClasses [&NiagaraSpriteRendererProperties]
  renderers [
   {rendererClass=&NiagaraSpriteRendererProperties}
   {rendererClass=&NiagaraSpriteRendererProperties, rendererIndex=1}
  ]
 }
 NE_MainRibbon001 {
  emitterSpawnScript [/SetVariables(Emitter.float=HlslExpression() {hlslExpression="User.ColorVariantID == 2 ? 1 : User.ColorVariantID == 3 ? 1 : 0"}) {isSetParameters=true}]
  emitterUpdateScript [
   &EmitterState()
   &VC_FilteredSocketCount()
   &VC_RateQualityHandler()
   &VC_SpawnRibbon_v1(Number of Ribbons=$Emitter.PortCount, SpawnRate=60)
  ]
  particleSpawnScript [
   &VC_PaintedVariantsOverride(
    Black=<rgba 0.026 0.026 0.026 1>
    BurntSienna=<rgba 0.625 0.1528 0.026 1>
    Cobalt=<rgba 0 1.192 5 1>
    Crimson=<rgba 1 0.015 0 1>
    DefaultColor=<rgba 3.3333 0 10 9.5>
    ForestGreen=<rgba 0.136 1.3 0 1>
    Lime=<rgba 1.0523 1.5 0.0087 1>
    Orange=<rgba 5 1.072 0 1>
    Pink=<rgba 8 0 8 1>
    R1 Black:LinearColor=<rgba 15 15 15 1>
    R1 BurntSienna:LinearColor=<rgba 6.875 0.4001 0 1>
    R1 Cobalt:LinearColor=<rgba 50 50 50 0>
    R1 Crimson:LinearColor=<rgba 0.4635 0 0.0126 0>
    R1 DefaultColor=<rgba 1 1 1 0>
    R1 ForestGreen:LinearColor=<rgba 0 0.5 0.0253 0>
    R1 Gold:LinearColor=<rgba 1 1 1 0>
    R1 Grey:LinearColor=<rgba 0.2031 0.2031 0.2031 0>
    R1 Lime:LinearColor=<rgba 0.2943 0.5885 0 0>
    R1 Orange:LinearColor=<rgba 0.6724 0.2346 0 0>
    R1 Pink:LinearColor=<rgba 0.5052 0.0605 0.3435 0>
    R1 Purple:LinearColor=<rgba 0.2789 0 0.6795 0>
    R1 Saffron:LinearColor=<rgba 0.9 0.69 0 0>
    R1 SkyBlue:LinearColor=<rgba 0 0.357 0.5365 0>
    R1 TitaniumWhite:LinearColor=<rgba 0.75 0.75 0.75 0>
    Saffron=<rgba 2 1.9015 0 1>
    SkyBlue=<rgba 0 1.6 1.6 1>
    TitaniumWhite=<rgba 2 2 2 1>
    Use R1 Painted Overrides=true
   )
   &VC_InitializeRibbon_v1(
    Color=$Particles.Color
    Custom Offset=true
    Enable Skeletal Position=true
    Lifetime=0.7
    Local Vector Offset=&AddVectorToPosition(
     Position=/Niagara/DynamicInputs/TypeConversions/MakeCustomPositionFromBool(Bool:NiagaraBool=$User.bIsFrontendPreview, True Position:NiagaraPosition=<12 0 0>, False Position:NiagaraPosition=<10 0 0>)
     Vector=$User.DTCarOffset
    )
    Offset Amount=1
    Ribbon Width=70
   )
   &AddVelocity(Velocity=<-1850 0 0>) {module=AddVelocity001}
  ]
  particleUpdateScript [
   &ParticleState()
   &InheritSourceMovement(
    Applied Position Scale=&Multiply_VectorByFloat(
     Vector=<1 1 1>
     Float=&FloatFromCurve(FloatCurve=DataInterface() {propertyValues {Curve {keys [{interpMode=RCIM_Cubic, tangentMode=RCTM_User, time=0.3, value=1}, {interpMode=RCIM_Cubic, tangentMode=RCTM_User, time=1, value=0}]}}})
    )
    ApplyForce=false
    ApplyPosition=true
    ApplyPositionFalloff=false
   ) {module=InheritSourceMovement001}
   &SolveForcesAndVelocity(Clamp Velocity=false, Limit Acceleration=false)
   &VC_UpdateRibbon_v1(Enable Skeletal Position=true)
   &ScaleRibbonWidth(Ribbon Width Scale=&FloatFromCurve(
     FloatCurve=DataInterface() {propertyValues {Curve {
        defaultValue=1
        keys [
         {interpMode=RCIM_Cubic, tangentMode=RCTM_User, time=0.4, value=1}
         {arriveTangent=-1.7941, interpMode=RCIM_Cubic, leaveTangent=-1.7941, time=0.5, value=0.8235}
         {arriveTangent=-0.8601, interpMode=RCIM_Cubic, leaveTangent=0.3194, tangentMode=RCTM_Break, time=0.6, value=0.6412}
         {arriveTangent=-2.2059, interpMode=RCIM_Cubic, leaveTangent=-2.2059, time=0.7, value=0.2882}
         {arriveTangent=-1.4412, interpMode=RCIM_Cubic, leaveTangent=-1.4412, time=0.8, value=0.2}
         {arriveTangent=-0.5355, interpMode=RCIM_Cubic, leaveTangent=-0.2251, tangentMode=RCTM_Break, time=0.9, value=0}
        ]
       }}}
     Scale Curve=1.4
    ))
   /SetVariables(Particles.RibbonWidth_1:NiagaraFloat=&Multiply_Float(
     A=50
     B=&FloatFromCurve(FloatCurve=DataInterface() {propertyValues {Curve {
         defaultValue=1
         keys [
          {interpMode=RCIM_Cubic, tangentMode=RCTM_User, time=0.2, value=1}
          {arriveTangent=-1.842, interpMode=RCIM_Cubic, leaveTangent=-1.842, tangentMode=RCTM_User, time=0.6, value=0}
         ]
        }}})
    )) {isSetParameters=true}
  ]
  rendererClasses [&NiagaraRibbonRendererProperties]
  renderers [
   {rendererClass=&NiagaraRibbonRendererProperties}
   {rendererClass=&NiagaraRibbonRendererProperties, rendererIndex=1}
  ]
 }
 NE_MainRibbon002 {
  emitterUpdateScript [
   &EmitterState()
   &VC_FilteredSocketCount() {enabled=false}
   &VC_RateQualityHandler() {enabled=false}
   &SpawnParticlesFromOtherEmitter(
    Attribute Reader=DataInterface() {propertyValues {EmitterBinding {bindingMode=Other, emitterName=NE_Ninjastar001}}}
    Calculate Spawn Rate Per Particle=false
    Spawn Rate=100
    Use Maximum Age=false
    Use Minimum Age=false
   )
   &VC_SpawnRibbon_v1(Number of Ribbons=$Emitter.PortCount, SpawnRate=1)
  ]
  particleSpawnScript [
   &VC_PaintedVariantsOverride(
    Black=<rgba 0.026 0.026 0.026 1>
    BurntSienna=<rgba 0.625 0.1528 0.026 1>
    Cobalt=<rgba 0 1.192 5 1>
    Crimson=<rgba 1 0.015 0 1>
    DefaultColor=<rgba 2.3333 0 7 15>
    ForestGreen=<rgba 0.136 1.3 0 1>
    Lime=<rgba 1.0523 1.5 0.0087 1>
    Orange=<rgba 5 1.072 0 1>
    Pink=<rgba 8 0 8 1>
    Saffron=<rgba 2 1.9015 0 1>
    SkyBlue=<rgba 0 1.6 1.6 1>
    TitaniumWhite=<rgba 2 2 2 1>
   )
   &InitializeParticle(
    Color=$Particles.Color
    Lifetime=0.25
    Ribbon Width=15
    Ribbon Width Mode=::NewEnumerator1
    UsePositionOffset=false
    Write Material Random=true
   )
   &SampleParticlesFromOtherEmitter(Particle ID Sampling=::NewEnumerator1)
   &VC_InitializeRibbon_v1(
    Color=$Particles.Color
    Custom Offset=true
    Enable Skeletal Position=true
    Local Vector Offset=&TransformVector(
     Vector=Linked() {linkedVariable {name=System.DeathPosition, type=NiagaraPosition}}
     Source Space=::Simulation
     Destination Space=::Local
     Transform Vector as Position=true
    )
    Offset Amount=1
    Ribbon Width=25
   ) {enabled=false}
   &AddVelocity(Velocity=<500 0 0>) {enabled=false, module=AddVelocity001}
  ]
  particleUpdateScript [
   &ParticleState()
   &InheritSourceMovement(
    Applied Position Scale=&Multiply_VectorByFloat(
     Vector=<1 1 1>
     Float=&FloatFromCurve(FloatCurve=DataInterface() {propertyValues {Curve {keys [{interpMode=RCIM_Cubic, tangentMode=RCTM_User, time=0.3, value=1}, {interpMode=RCIM_Cubic, tangentMode=RCTM_User, time=1, value=0}]}}})
    )
    ApplyForce=false
    ApplyPosition=true
    ApplyPositionFalloff=false
   ) {module=InheritSourceMovement001}
   &SolveForcesAndVelocity(Clamp Velocity=false, Limit Acceleration=false)
   &VC_UpdateRibbon_v1(Enable Skeletal Position=true) {enabled=false}
  ]
  rendererClasses [&NiagaraRibbonRendererProperties]
  renderers [{rendererClass=&NiagaraRibbonRendererProperties}]
 }
 NE_Ninjastar001 {
  emitterUpdateScript [
   &EmitterState()
   &VC_FilteredSocketCount()
   &SpawnRate(
    SpawnRate=&RandomRangeFloat(Minimum=5, Maximum=6, Recalculate Random Each Loop=false)
    Use Spawn Probability=false
   )
   &SpawnRate(
    SpawnGroup=1
    SpawnRate=&RandomRangeFloat(Minimum=4, Maximum=6, Recalculate Random Each Loop=false)
    Use Spawn Probability=false
   ) {module=SpawnRate001}
   &self:ScratchModule_04(
    DeathColorArray:NiagaraDataInterfaceArrayColor=$System.DeathColorArray
    DeathPositionsArray:NiagaraDataInterfaceArrayPosition=$System.DeathArray
   ) {module=ScratchModule_04}
  ]
  particleSpawnScript [
   &VC_PaintedVariantsOverride(
    Black=<rgba 3 3 3 1>
    BurntSienna=<rgba 2 1.614 1.5104 1>
    Cobalt=<rgba 3 3 3 1>
    Crimson=<rgba 3 2.2656 2.2656 1>
    DefaultColor=<rgba 2.4826 1.224 5 1>
    ForestGreen=<rgba 2.4844 3 2.5104 1>
    Lime=<rgba 2.8281 3 2.6562 1>
    Orange=<rgba 5 3.7112 3.0208 1>
    Pink=<rgba 3 2.5781 2.8466 1>
    Saffron=<rgba 5 4.0885 1.0938 1>
    SkyBlue=<rgba 2.0312 2.6758 3 1>
    TitaniumWhite=<rgba 2 2 2 1>
   )
   &InitializeParticle(
    Color Mode=::NewEnumerator0
    Lifetime Max=0.8
    Lifetime Min=0.27
    Lifetime Mode=::NewEnumerator1
    Mesh Scale Mode=::NewEnumerator1
    Mesh Uniform Scale Max=1.5
    Mesh Uniform Scale Min=1.3
    Sprite Size Mode=::NewEnumerator3
    UsePositionOffset=true
    Write Material Random=true
   )
   &PartitionParticles(Number Of Partitions=$Emitter.PortCount)
   &SkeletalMeshLocation(
    Bone / Socket Index=$Particles.PartitionParticles.Partition
    Bone / Socket Sampling Mode=::NewEnumerator5
    MaskSpawnGroup=false
    Mesh Sampling Type=::NewEnumerator6
    Sampled Position Offset=&AddVectorToPosition(Position=<-30 0 0>, Vector=$User.DTCarOffset)
    Skeletal Mesh=$System.SkeletalMesh
   )
   &AddVelocity(
    Cone Axis=<1 0 -0.4>
    Inner Cone Angle=20
    Module.UseConeFalloff=false
    Module.UseVelDistribution=false
    Override Seed=false
    Velocity Mode=::NewEnumerator2
    Velocity Speed=&RandomRangeFloat(Minimum=-800, Maximum=-900)
   )
   &InitialMeshOrientation(
    Mesh Orientation Mode=::NewEnumerator7
    Rotation=&RandomRangeVector(
     Minimum=&SelectVectorFromArray(
      Direct Array Index=&SpawnGroup()
      Vector Selection Array=DataInterface() {propertyValues {FloatData [<0 0 0>, <0 0 0>]}}
     )
     Maximum=&SelectVectorFromArray(
      Direct Array Index=&SpawnGroup()
      Vector Selection Array=DataInterface() {propertyValues {FloatData [<3 3 0>, <0 0 0>]}}
     )
    )
    Use Rotation Vector=true
   )
   /SetVariables(Particles.Visibility:NiagaraInt32=$Emitter.SpawnGroup) {isSetParameters=true}
  ]
  particleUpdateScript [
   &ParticleState()
   &UpdateMeshOrientation(
    Rotation Rate=0.5
    Rotation Vector=&SelectVectorFromArray(
     Direct Array Index=&SpawnGroup()
     Vector Selection Array=DataInterface() {propertyValues {FloatData [<0 5 0>, <0 0 5>]}}
    )
   )
   &Collision(
    Age Colliding Particles=false
    Analytical Collision Plane Position 1:Vector3f=<0 0 50>
    Analytical Collision Plane Position 2:Vector3f=<0 0 50>
    Analytical Collision Plane Space:ENiagaraCoordinateSpace=::World
    CPU Collision Type=::NewEnumerator3
    Friction=0.6
    KillOnCollision=true
    Module.Control Rest State Via Surface Normal=false
    Module.EnableMaxCollisionCount=false
    Module.Randomize Collision Normal Vector=false
    Restitution=0.1
   )
   &GravityForce(Gravity=&RandomRangeVector(Minimum=<0 0 -950>, Maximum=<0 0 -1600>))
   &InheritSourceMovement(
    Applied Position Scale=&Multiply_VectorByFloat(
     Vector=<1 1 1>
     Float=&FloatFromCurve(FloatCurve=DataInterface() {propertyValues {Curve {keys [{interpMode=RCIM_Cubic, tangentMode=RCTM_User, time=0.3, value=1}, {interpMode=RCIM_Cubic, tangentMode=RCTM_User, time=1, value=0}]}}})
    )
    ApplyForce=false
    ApplyPosition=true
    ApplyPositionFalloff=false
   )
   &SolveForcesAndVelocity(Clamp Velocity=false, Limit Acceleration=false)
   &DynamicMaterialParameters(
    Index 0 Param 1=1
    Index 0 Param 2=1
    Module.Index0Float0Write=true
    Module.Index0Float1Write=true
    Module.Index0Float2Write=true
    Module.Index0Float3Write=true
   )
   &self:WriteDeathArrayScratch() {module=WriteDeathArrayScratch}
  ]
  rendererClasses [&NiagaraMeshRendererProperties, &NiagaraSpriteRendererProperties]
  renderers [
   {rendererClass=&NiagaraMeshRendererProperties, rendererIndex=0}
   {rendererClass=&NiagaraMeshRendererProperties, rendererIndex=1}
   {rendererClass=&NiagaraSpriteRendererProperties, rendererIndex=2}
  ]
 }
 NE_ThrusterSparks001 {
  emitterUpdateScript [
   &EmitterState()
   &VC_FilteredSocketCount()
   &SpawnRate(SpawnRate=&RandomRangeFloat(Minimum=15, Maximum=18), Use Spawn Probability=false)
  ]
  particleSpawnScript [
   &VC_PaintedVariantsOverride(
    Black=<rgba 0.0885 0.0885 0.0885 1>
    BurntSienna=<rgba 0.474 0.1435 0 1>
    Cobalt=<rgba 0 0.0925 5 1>
    Crimson=<rgba 5 0 0 1>
    DefaultColor=<rgba 3.9236 0.8854 10 1>
    ForestGreen=<rgba 0.0346 0.3958 0 1>
    Gold:LinearColor=<rgba 0.3646 0.3646 0.3646 1>
    Lime=<rgba 0.3579 0.4696 0 1>
    Pink=<rgba 3 0.3594 2.0398 1>
    Saffron=<rgba 0.9478 1.5 0 1>
    SkyBlue=<rgba 0 1.3308 2 1>
    TitaniumWhite=<rgba 50 50 50 1>
   )
   &InitializeParticle(
    Color Mode=::NewEnumerator0
    Lifetime Max=0.18
    Lifetime Min=0.13
    Lifetime Mode=::NewEnumerator1
    Sprite Rotation Angle=&RandomRangeFloat(Minimum=180, Maximum=-180)
    Sprite Rotation Mode=::NewEnumerator0
    Sprite Size Max=<12 70>
    Sprite Size Min=<10 50>
    Sprite Size Mode=::NewEnumerator2
    UsePositionOffset=false
    Write Material Random=true
   )
   &PartitionParticles(Number Of Partitions=$Emitter.PortCount)
   &SkeletalMeshLocation(
    Bone / Socket Index=$Particles.PartitionParticles.Partition
    Bone / Socket Sampling Mode=::NewEnumerator5
    MaskSpawnGroup=false
    Mesh Sampling Type=::NewEnumerator6
    Sampled Position Offset=&AddVectorToPosition(Position=<0 0 0>, Vector=$User.DTCarOffset)
    Skeletal Mesh=$System.SkeletalMesh
   )
   &AddVelocity(
    Cone Angle=100
    Module.UseConeFalloff=false
    Module.UseVelDistribution=false
    Override Seed=false
    Velocity Mode=::NewEnumerator2
    Velocity Speed=&RandomRangeFloat(Minimum=-900, Maximum=-1000)
   )
   &SubUVAnimation(
    "Sprite Renderer "=DataInterface() {propertyValues {SpriteRenderer=&self:NE_ThrusterSparks001_0.NiagaraSpriteRendererProperties_0}}
    SubUV Animation Mode:ENiagara_SubUVLookupModeV2=::NewEnumerator2
    UseEndFrame=false
    UseStartFrame=false
   )
  ]
  particleUpdateScript [
   &ParticleState()
   &ScaleSpriteSize(Uniform Curve Sprite Scale=DataInterface() {propertyValues {Curve {
       defaultValue=0
       keys [
        {interpMode=RCIM_Cubic, tangentMode=RCTM_User, time=0.3, value=1}
        {arriveTangent=-2.6631, interpMode=RCIM_Cubic, leaveTangent=-2.6631, tangentMode=RCTM_User, time=1, value=0}
       ]
      }}}) {module=ScaleSpriteSize001}
   &CurlNoiseForce(
    Bias Noise=false
    Noise Frequency=150
    Noise Strength=&RandomRangeFloat(Minimum=100, Maximum=150)
    Pan Noise=false
   )
   &InheritSourceMovement(
    Applied Position Scale=&Multiply_VectorByFloat(
     Vector=<1 1 1>
     Float=&FloatFromCurve(FloatCurve=DataInterface() {propertyValues {Curve {keys [{interpMode=RCIM_Cubic, tangentMode=RCTM_User, time=0.3, value=1}, {interpMode=RCIM_Cubic, tangentMode=RCTM_User, time=1, value=0}]}}})
    )
    ApplyForce=false
    ApplyPosition=true
    ApplyPositionFalloff=false
   )
   &SolveForcesAndVelocity(Clamp Velocity=false, Limit Acceleration=false)
   &ScaleColor(Scale RGB=&VectorFromFloat(Value=2), ScaleA=true, ScaleRGB=true)
  ]
  rendererClasses [&NiagaraSpriteRendererProperties]
  renderers [{rendererClass=&NiagaraSpriteRendererProperties}]
 }
}
summary {
 systemName=NS_Boost_Ninjastar
 emitters [
  {emitterName=NE_MainRibbon001, rendererClasses [&NiagaraRibbonRendererProperties]}
  {emitterName=NE_ThrusterSparks001, rendererClasses [&NiagaraSpriteRendererProperties]}
  {
   emitterName=NE_Ninjastar001
   rendererClasses [&NiagaraMeshRendererProperties, &NiagaraSpriteRendererProperties]
  }
  {emitterName=NE_MainRibbon002, rendererClasses [&NiagaraRibbonRendererProperties]}
  {emitterName=DeathParticles, rendererClasses [&NiagaraSpriteRendererProperties]}
 ]
 userVariables [
  User.bIsFrontendPreview {type=NiagaraBool}
  User.bIsPainted {type=NiagaraBool}
  User.ColorVariantID {type=NiagaraInt32}
  User.Desaturation {type=NiagaraFloat, defaultValue=1, description=""}
  User.DTCarOffset {type=Vector3f, description="", defaultValue {struct=Vector3f, value=<0 0 0>}}
  User.PaintedColor {type=LinearColor, description="", defaultValue {struct=LinearColor, value=<rgba 1 1 1 1>}}
  User.PaintedScalar {type=NiagaraFloat, defaultValue=1, description=""}
  User.PortCount {type=NiagaraInt32, defaultValue=1}
 ]
}
