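# Unlit additive material for Niagara ribbon needle trails in the Ninjastar vehicle boost.
# A procedural UV mask ramps up quickly along U by Sharpness, fades toward the far end and tapers at both V edges.
# Particle color tints the mask, and the addDetail static switch adds panning, biased cloud-distort noise with a squared contrast curve.
# The result feeds the shared MF_FX_EmissiveAdjustments function.
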Material /VehicleCosmetics/Boosts/Boost_Ninjastar/Materials/M_NinjaStart_NeedleTrail
graph {
 properties {
  bAutomaticallySetUsageInEditor=false
  bUsedWithNiagaraRibbons=true
  blendMode=BLEND_Additive
  shadingModel=MSM_Unlit
  usedShadingModels=Unlit
 }
 nodes [
  MaterialExpressionVertexColor_0 = ParticleColor()
  n1 = TextureCoordinate()
  n2 = ComponentMask(None=@n1) {r=true}
  n3 = Constant() {r=12}
  n4 = ScalarParameter() {defaultValue=1, parameterName=Sharpness}
  MaterialExpressionConstantClamp_0 = Saturate(None=Multiply(A=@n2, B=Multiply(A=@n3, B=@n4)))
  n5 = ComponentMask(None=@n1) {g=true}
  n6 = Multiply(A=Multiply(A=OneMinus(None=@n2), B=@MaterialExpressionConstantClamp_0), B=Multiply(A=@n5, B=OneMinus(None=@n5)))
  n7 = TextureSample(UVs=Panner(Coordinate=TextureCoordinate() {vTiling=0.5}) {speedX=-2}) {texture=/DealershipSharedLibrary/FX/Textures/Noise/T_VCVFX_CloudDistort_A}
  n8 = Multiply(A=@MaterialExpressionVertexColor_0.RGB, B=Multiply(A=@n6, B=ConstantBiasScale(None=@n7.R) {scale=1}))
  n9 = Constant() {r=2}
  MaterialFunctionCall(Emissive=StaticSwitchParameter(True=Power(Base=@n8, Exp=@n9), False=Multiply(A=@MaterialExpressionVertexColor_0.RGB, B=@n6)) {parameterName=addDetail}) {materialFunction=&MF_FX_EmissiveAdjustments}
 ]
}
