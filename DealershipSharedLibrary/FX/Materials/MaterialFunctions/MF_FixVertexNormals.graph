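# Material function that rebuilds surface normals in tangent space and combines them with an input normal map.
# Custom HLSL ddx/ddy of world position feed a cross product for faceted normals.
# The SmoothNormals static switch substitutes normalized absolute world position for that cross product.
# It outputs the rebuilt normal as FixedNormlas, plus AdditionalNormals, a reoriented-normal-mapping style blend with the input Normal.
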MaterialFunction /DealershipSharedLibrary/FX/Materials/MaterialFunctions/MF_FixVertexNormals
graph {nodes [
  n1 = WorldPosition()
  MaterialExpressionWorldPosition_0 = WorldPosition()
  n2 = Custom!(x=@MaterialExpressionWorldPosition_0.XYZ) {
   code="return ddx(x);"
   inputs [{
     inputName=x
     input {
      expression=&self:MaterialExpressionWorldPosition_0
      mask=1
      maskB=1
      maskG=1
      maskR=1
     }
    }]
  }
  n3 = Custom!(y=@MaterialExpressionWorldPosition_0.XYZ) {
   code="return ddy(y);"
   inputs [{
     inputName=y
     input {
      expression=&self:MaterialExpressionWorldPosition_0
      mask=1
      maskB=1
      maskG=1
      maskR=1
     }
    }]
  }
  n4 = CrossProduct(A=@n2, B=@n3)
  n5 = FunctionInput(Preview=StaticBool()) {bUsePreviewValueAsDefault=true, inputName=SmoothNormals, inputType=FunctionInput_StaticBool}
  n6 = Transform(None=Normalize(VectorInput=StaticSwitch(True=@n1.XYZ, False=@n4, Value=@n5))) {transformSourceType=TRANSFORMSOURCE_World, transformType=TRANSFORM_Tangent}
  FunctionOutput(None=@n6) {outputName=FixedNormlas}
  n7 = Multiply(A=ConstantBiasScale(None=FunctionInput() {inputName=Normal}), B=Constant3Vector() {constant=<rgba -2 -2 2 0>})
  n8 = Add(A=@n7, B=Constant3Vector() {constant=<rgba -1 -1 0 0>})
  n9 = Add(A=Multiply(A=ConstantBiasScale(None=@n6), B=Constant3Vector() {constant=<rgba 2 2 2 0>}), B=Constant3Vector() {constant=<rgba 1 1 -1 0>})
  n10 = DotProduct(A=@n9, B=@n7)
  n11 = ComponentMask(None=@n9) {b=true}
  FunctionOutput(None=Subtract(A=Multiply(A=@n8, B=@n10), B=Multiply(A=@n11, B=@n9))) {outputName=AdditionalNormals}
 ]}
