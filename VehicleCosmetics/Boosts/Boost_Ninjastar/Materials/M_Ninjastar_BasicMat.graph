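# Unlit masked Niagara material for the Ninjastar boost cosmetic, textured with the hat shuriken diffuse and normal maps.
# It fakes lighting from a fixed direction, using a flat normal for diffuse and half-vector specular, plus normal-mapped Fresnel.
# Particle color tints a contrast-boosted grayscale diffuse, added as glow through MF_FX_EmissiveAdjustments.
# A pivot-movement static switch can blend the normal map toward MF_FixVertexNormals output by a UV3 value.
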Material /VehicleCosmetics/Boosts/Boost_Ninjastar/Materials/M_Ninjastar_BasicMat
graph {
 properties {
  bAutomaticallySetUsageInEditor=false
  bUsedWithNiagaraMeshParticles=true
  bUsedWithNiagaraRibbons=true
  bUsedWithNiagaraSprites=true
  blendMode=BLEND_Masked
  refractionMethod=RM_2DOffset
  shadingModel=MSM_Unlit
  usedShadingModels=Unlit
 }
 nodes [
  n1 = Constant3Vector() {constant=<rgba 0 0 1 0>}
  n2 = Transform(None=Normalize(VectorInput=Constant3Vector() {constant=<rgba -0.35 0.35 1 0>})) {transformSourceType=TRANSFORMSOURCE_World, transformType=TRANSFORM_Tangent}
  MaterialExpressionConstantClamp_0 = Saturate(None=ConstantBiasScale(None=DotProduct(A=@n1, B=@n2)))
  n3 = Power(Base=@MaterialExpressionConstantClamp_0, Exp=Constant() {r=0.2})
  n4 = TextureSampleParameter2D() {parameterName=Diffuse, texture=/VehicleCosmetics/Boosts/Boost_Ninjastar/Textures/T_Boost_Ninjastar_HatShuriken_A}
  n5 = CameraVectorWS()
  n6 = DotProduct(A=@n1, B=Normalize(VectorInput=Add(A=@n2, B=@n5)))
  n7 = Constant() {r=250}
  n8 = Add(A=Multiply(A=ConstantBiasScale(None=@n3) {bias=0.1, scale=2}, B=@n4.RGB), B=Multiply(A=Constant() {r=4}, B=Power(Base=@n6, Exp=@n7)))
  n9 = TextureSampleParameter2D() {
   parameterName=Normal
   samplerType=SAMPLERTYPE_Normal
   texture=/VehicleCosmetics/Boosts/Boost_Ninjastar/Textures/T_Boost_Ninjastar_HatShuriken_N
  }
  n10 = Add(A=@n8, B=Multiply(A=@n8, B=Fresnel(Normal=@n9.RGB)))
  MaterialExpressionMeshEmitterVertexColor_0 = ParticleColor()
  n11 = Constant() {r=2.5}
  n12 = Power(Base=Desaturation(None=@n4.RGB, Fraction=Constant() {r=1}) {luminanceFactors=<rgba 0.3 0.59 0.11 0>}, Exp=@n11)
  n13 = Multiply(
   A=@MaterialExpressionMeshEmitterVertexColor_0.RGB
   B=Multiply(A=Multiply(A=@n12, B=@n11), B=ScalarParameter() {defaultValue=1, parameterName=EmissiveScale})
  )
  MaterialFunctionCall(Emissive=Multiply(A=Add(A=@n10, B=@n13)) {constB=0.1}) {materialFunction=&MF_FX_EmissiveAdjustments}
  n14 = StaticBool() {value=true}
  MaterialExpressionConstantClamp_1 = Saturate(None=ComponentMask(None=TextureCoordinate() {coordinateIndex=3}) {r=true})
  StaticSwitchParameter(
   True=LinearInterpolate(
    A=@n9.RGB
    B=MaterialFunctionCall!(Normal=@n9.RGB, SmoothNormals=@n14) {materialFunction=/DealershipSharedLibrary/FX/Materials/MaterialFunctions/MF_FixVertexNormals}.FixedNormlas
    Alpha=@MaterialExpressionConstantClamp_1
   )
   False=@n9.RGB
  ) {parameterName="Movement(UsePreDefinedPivots)"}
 ]
}
