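# Translucent unlit Niagara particle material for explosion smoke in the Ninjastar vehicle boost cosmetic.
# A SubUV smoke texture is desaturated, tinted by particle color and multiplied by dimmed, panning cloud noise on tiled, fractional UVs, then passed through an emissive-adjustment function.
# Opacity combines texture alpha, particle alpha and a depth fade that softens intersections with scene geometry.
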Material /VehicleCosmetics/Boosts/Boost_Ninjastar/Materials/M_Ninjastar_OctoberExpoSmoke
graph {
 properties {
  bAutomaticallySetUsageInEditor=false
  bUsedWithNiagaraMeshParticles=true
  bUsedWithNiagaraSprites=true
  blendMode=BLEND_Translucent
  shadingModel=MSM_Unlit
  usedShadingModels=Unlit
 }
 nodes [
  n1 = Frac(None=TextureCoordinate() {uTiling=8, vTiling=8})
  n2 = TextureSample(UVs=Panner(Coordinate=Divide(A=@n1)) {speedX=0.5, speedY=0.2}) {texture=/DealershipSharedLibrary/FX/Textures/Noise/T_VCVFX_CloudDistort_A}
  n3 = ParticleColor()
  n4 = TextureSampleParameterSubUV!(UVs, Apply View MipBias) {parameterName=Texture, texture=/VehicleCosmetics/Boosts/Boost_Ninjastar/Textures/T_Boost_Ninjastar_ExplosionSmoke_F}
  n5 = Desaturation(None=@n4.RGB) {luminanceFactors=<rgba 0.7152 0 0.3737 1>}
  MaterialFunctionCall(Emissive=Multiply(A=Multiply(A=@n2.RGB) {constB=0.5}, B=Multiply(A=@n3.RGB, B=@n5))) {materialFunction=&MF_FX_EmissiveAdjustments}
  Multiply(
   A=Multiply(A=@n4.A, B=@n3.A)
   B=MaterialFunctionCall(Opacity In=Constant() {r=1}, DepthFade Distance=Constant() {r=100}) {materialFunction=&MF_VCVFX_DepthFade}.Result
  )
 ]
}
