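# M_Ninjastar_Trail is an unlit translucent Niagara material for the Ninjastar boost's trail, usable on sprites, ribbons and meshes.
# A noise-warped, panning energy texture, masked by UV edge fades and tinted by particle color, drives the emissive.
# Opacity reuses that mask with particle alpha, cut out by inverted speed-trail streaks, with optional sine-timed on/off flicker.
# A scrolling sine wave, weighted along U, supplies local-space vertex wobble for world position offset.
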Material /VehicleCosmetics/Boosts/Boost_Ninjastar/Materials/M_Ninjastar_Trail
graph {
 properties {
  bAutomaticallySetUsageInEditor=false
  bUsedWithNiagaraMeshParticles=true
  bUsedWithNiagaraRibbons=true
  bUsedWithNiagaraSprites=true
  blendMode=BLEND_Translucent
  shadingModel=MSM_Unlit
  usedShadingModels=Unlit
 }
 nodes [
  n1 = Multiply(A=OneMinus(None=Time()), B=ScalarParameter() {defaultValue=2.5, group=WPO, parameterName=Speed})
  n2 = Divide(A=ComponentMask(None=TextureCoordinate()) {r=true}, B=ScalarParameter() {defaultValue=1, group=WPO, parameterName=tile})
  Cosine(None=Add(A=@n1, B=@n2)) {period=0.7}
  n3 = Time()
  n4 = TextureSample(UVs=Panner(Coordinate=TextureCoordinate() {uTiling=3}, Time=@n3) {speedX=-0.03, speedY=0.07}) {texture=/DealershipSharedLibrary/FX/Textures/Noise/T_VCVFX_SS_Windtunnel_Noise_A}
  n5 = ScalarParameter() {defaultValue=0.025, parameterName=WrapX}
  n6 = ScalarParameter() {defaultValue=0.05, parameterName=WrapY}
  n7 = Add(
   A=Multiply(A=@n4.R, B=MaterialFunctionCall(X=@n5, Y=@n6) {materialFunction=&MakeFloat2}.Result)
   B=TextureCoordinate()
  )
  n8 = ConstantBiasScale(None=TextureSample(UVs=Panner(Coordinate=@n7, Time=@n3) {speedX=-2}) {texture=/VehicleCosmetics/Boosts/Boost_Ninjastar/Textures/T_Boost_Ninjastar_EnergyTrail_RGB}.R) {bias=-0.01, scale=5}
  n9 = TextureCoordinate()
  n10 = ComponentMask(None=@n9) {g=true}
  n11 = OneMinus(None=@n10)
  n12 = Saturate(None=Multiply(A=Multiply(A=@n10, B=@n11), B=Constant() {r=3}))
  n13 = ComponentMask(None=@n9) {r=true}
  n14 = Saturate(None=OneMinus(None=@n13))
  n15 = Multiply(A=@n14, B=Saturate(None=@n13))
  n16 = Constant() {r=10}
  n17 = Multiply(A=Multiply(A=Multiply(A=@n12, B=@n14), B=Multiply(A=@n15, B=@n16)), B=Constant() {r=5})
  n18 = Saturate(None=Multiply(B=Multiply(A=@n8, B=@n17)) {constA=2})
  n19 = ParticleColor()
  MaterialFunctionCall(Emissive=Multiply(A=@n18, B=@n19.RGB)) {materialFunction=&MF_FX_EmissiveAdjustments}
  n20 = Time()
  n21 = Sine(None=@n20)
  n22 = Sine(None=@n20)
  n23 = ScalarParameter() {defaultValue=1.5, parameterName=SpeedFlickering}
  n24 = MaterialFunctionCall(Speed=Multiply(A=Multiply(A=@n21) {constB=2}, B=Multiply(A=@n22, B=@n23) {constB=2})) {materialFunction=&TimeWithSpeedVariable}
  n25 = Multiply(A=Add(A=Sine(None=@n24.Time))) {constB=0.5}
  n26 = Saturate(None=ScalarParameter() {parameterName=Flicker})
  n27 = Constant() {r=5}
  n28 = TextureCoordinate() {vTiling=0.33}
  n29 = Multiply(A=Constant2Vector() {g=-0.05, r=-0.025}, B=@n4.R)
  n30 = TextureSample(UVs=Panner(Coordinate=Add(A=@n28, B=@n29), Time=@n3) {speedX=-2}) {texture=/VehicleCosmetics/Boosts/Boost_Ninjastar/Textures/T_Boost_Ninjastar_SpeedTrails_RGB}
  n31 = Saturate(None=Power(Base=Multiply(A=@n27, B=@n30.G), Exp=Constant() {r=5}))
  Multiply(A=LinearInterpolate(B=Round(None=@n25), Alpha=Round(None=@n26)) {constA=1}, B=Multiply(A=OneMinus(None=@n31), B=Multiply(A=@n18, B=@n19.A)))
  n32 = Sine(None=Add(A=@n2, B=@n1)) {period=0.2}
  n33 = AppendVector(A=Constant(), B=@n32)
  n34 = Multiply(
   A=Multiply(A=AppendVector(A=@n33, B=@n32), B=ScalarParameter() {defaultValue=20, group=WPO, parameterName=Intensity})
   B=ScalarParameter() {defaultValue=1, group=WPO, parameterName=Amount}
  )
  n35 = Power(Base=ComponentMask(None=TextureCoordinate()) {r=true}, Exp=ScalarParameter() {defaultValue=2, parameterName=PowerMiddle2})
  n36 = ScalarParameter() {defaultValue=0.05, parameterName=StaticPoint}
  Transform(None=Multiply(A=ConstantBiasScale(None=@n34), B=Subtract(A=@n35, B=@n36))) {transformSourceType=TRANSFORMSOURCE_Local}
 ]
}
